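; Projectile Blueprint for the secondary ranged fire of the FringePlank Firework item.
; On BeginPlay it calls the parent, then stops the looping audio tell component as a fix for FORT-483105.
; The overlap and tick events are empty, and the construction script only calls the parent.
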(event EventBeginPlay
  (|CallParentFunction)
  ;[1 1 1] Fixes FORT-483105
  (Audio|Components|Audio|Stop (Variables|Default|GetLoopingAudioTell)))

(event Collision|EventActorBeginOverlap
  (params (OtherActor :type Actor%20Object%20Reference)))

(event EventTick
  (params (DeltaSeconds :type Float%20%28single-precision%29)))

(fn ConstructionScript
  (|CallParentFunction))
